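# Fort_Events_Priority_Medium is a SoundClass, named as the medium-priority tier for event sounds in BlastBerry map content.
# It is a child of the Fort_Events sound class and shows no overrides or child classes of its own.
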SoundClass /BlastBerryMapContent/Game/Athena/Sounds/Events/SoundClasses/Fort_Events_Priority_Medium
parent=&Fort_Events
